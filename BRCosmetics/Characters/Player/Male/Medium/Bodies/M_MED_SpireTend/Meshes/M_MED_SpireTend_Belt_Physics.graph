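# Physics asset for the SpireTend male medium body's belt, defining physical front and back dynamic belt chains.
# Each chain runs three capsule segments from a tiny master base body, joined by narrow swing and twist limits that generally tighten toward the tip.
# Two mirrored constraints with locked angular motion tie the chain tips together, closing a loop.
# Capsules on the pelvis and left thigh twist bone provide body shapes beside the chains.
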PhysicsAsset /BRCosmetics/Characters/Player/Male/Medium/Bodies/M_MED_SpireTend/Meshes/M_MED_SpireTend_Belt_Physics
bodies {
 base_dyn_beltChain_bk_MSTR [{shapeType=Capsule, length=0.505, radius=0.505, shapeName=base_dyn_beltChain_bk_MSTR_capsule, rotation=<rot 0 -90 0>}]
 base_dyn_beltChain_ft_MSTR [{shapeType=Capsule, length=0.505, radius=0.505, shapeName=base_dyn_beltChain_ft_MSTR_capsule, rotation=<rot 0 -90 0>}]
 dyn_beltChain_bk_00 [{shapeType=Capsule, length=8, radius=1.7, shapeName=dyn_beltChain_ft_00_capsule, center=<5.25 0 0.1>, rotation=<rot 3 -87 -95>}]
 dyn_beltChain_bk_01 [{shapeType=Capsule, length=6, radius=1.7, shapeName=dyn_beltChain_bk_01_capsule, center=<5.75 0 0>, rotation=<rot 0 90 -89>}]
 dyn_beltChain_bk_02 [{shapeType=Capsule, length=10, radius=1.5, shapeName=dyn_beltChain_bk_02_capsule, center=<7.5 -1 -4.5>, rotation=<rot 0 -107.5 -48>}]
 dyn_beltChain_ft_00 [{shapeType=Capsule, length=8, radius=1.7, shapeName=dyn_beltChain_ft_00_capsule, center=<5.25 0 0.1>, rotation=<rot 0 -93 -85.5>}]
 dyn_beltChain_ft_01 [{shapeType=Capsule, length=6, radius=1.7, shapeName=dyn_beltChain_ft_01_capsule, center=<3.8 -0.15 0.3>, rotation=<rot 0 91 -89>}]
 dyn_beltChain_ft_02 [{shapeType=Capsule, length=10, radius=1.5, shapeName=dyn_beltChain_ft_02_capsule, center=<5 1.5 -4>, rotation=<rot -22.25 -90 -35.25>}]
 pelvis [{shapeType=Capsule, length=9.25, radius=12.25, center=<0 -1.5875 -1.4012>}]
 thigh_twist_01_l [{shapeType=Capsule, length=18.5, radius=12, center=<5.4029 -0.575 -2.5>, rotation=<rot 85 -180 0>}]
}
constraints [
 {bone1Name=dyn_beltChain_bk_00, bone2Name=base_dyn_beltChain_bk_MSTR, swing1LimitDegrees=20, swing2LimitDegrees=10, twistLimitDegrees=5}
 {bone1Name=dyn_beltChain_ft_00, bone2Name=base_dyn_beltChain_ft_MSTR, swing1LimitDegrees=15, swing2LimitDegrees=20, twistLimitDegrees=5}
 {bone1Name=dyn_beltChain_bk_01, bone2Name=dyn_beltChain_bk_00, swing1LimitDegrees=15, swing2LimitDegrees=15, twistLimitDegrees=5}
 {bone1Name=dyn_beltChain_bk_02, bone2Name=dyn_beltChain_bk_01, swing1LimitDegrees=10, swing2LimitDegrees=10, twistLimitDegrees=5}
 {bone1Name=dyn_beltChain_ft_01, bone2Name=dyn_beltChain_ft_00, swing1LimitDegrees=15, swing2LimitDegrees=15, twistLimitDegrees=5}
 {bone1Name=dyn_beltChain_ft_02, bone2Name=dyn_beltChain_ft_01, swing1LimitDegrees=10, swing2LimitDegrees=10, twistLimitDegrees=5}
 {
  bone1Name=dyn_beltChain_bk_02
  bone2Name=dyn_beltChain_ft_02
  swing1LimitDegrees=5
  swing1Motion=Locked
  swing2LimitDegrees=5
  swing2Motion=Locked
  twistLimitDegrees=5
  twistMotion=Locked
 }
 {
  bone1Name=dyn_beltChain_ft_02
  bone2Name=dyn_beltChain_bk_02
  swing1LimitDegrees=5
  swing1Motion=Locked
  swing2LimitDegrees=5
  swing2Motion=Locked
  twistLimitDegrees=5
  twistMotion=Locked
 }
]
